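# Translucent unlit Niagara particle material that fakes a bubble decal on scene geometry, with depth test disabled for the LimeRock treehouse effect.
# It projects from world position behind translucency relative to the particle, blending three atan2 angles by scene normal into polar UVs.
# Panning streaks and Fresnel-driven hue shift are shaped by a gradient-remapped sphere mask, with MeshScale and Erode dynamic parameters.
# Forward and mobile paths drop streaks and hue shift.
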Material /LimeRock_CountdownTimer/Assets/FX/Treehouse/PreEvent_Stage01/LimeRock/Assets/FX/Treehouse_Bubble/M_LimeRock_Bubble_Decal
graph {
 properties {
  bDisableDepthTest=true
  bUsedWithNiagaraMeshParticles=true
  bUsedWithNiagaraSprites=true
  blendMode=BLEND_Translucent
  pixelDepthOffsetMode=PDOM_Legacy
  shadingModel=MSM_Unlit
  usedShadingModels=Unlit
 }
 nodes [
  ActorPositionWS()
  n1 = ParticleColor()
  n2 = Subtract(
   A=MaterialFunctionCall() {materialFunction=&WorldPositionBehindTranslucency}.Result
   B=ParticlePositionWS()
  )
  MaterialExpressionMaterialFunctionCall_12 = MaterialFunctionCall(Float3=@n2) {materialFunction=&BreakOutFloat3Components}
  n3 = Custom!(x=@MaterialExpressionMaterialFunctionCall_12.R, y=@MaterialExpressionMaterialFunctionCall_12.R) {
   code="return (atan2(y,x));"
   description=atan2
   outputType=CMOT_Float1
   inputs [
    {
     inputName=x
     input {expression=&self:MaterialExpressionMaterialFunctionCall_12}
    }
    {
     inputName=y
     input {
      expression=&self:MaterialExpressionMaterialFunctionCall_12
      outputIndex=2
     }
    }
   ]
  }
  n4 = Custom(x=@MaterialExpressionMaterialFunctionCall_12.G, y=@MaterialExpressionMaterialFunctionCall_12.G) {
   code="return (atan2(y,x));"
   description=atan2
   outputType=CMOT_Float1
   inputs [
    {
     inputName=x
     input {
      expression=&self:MaterialExpressionMaterialFunctionCall_12
      outputIndex=1
     }
    }
    {
     inputName=y
     input {
      expression=&self:MaterialExpressionMaterialFunctionCall_12
      outputIndex=2
     }
    }
   ]
  }
  n5 = SceneTexture() {sceneTextureId=PPI_WorldNormal}
  n6 = MaterialFunctionCall(Float3=NamedRerouteDeclaration(None=ComponentMask(None=@n5.Color) {b=true, g=true, r=true}) {name=WorldNormal, nodeColor=<rgba 0 1 0.2969 1>}) {materialFunction=&BreakOutFloat3Components}
  n7 = ScalarParameter() {defaultValue=0.7, parameterName=WorldAlignedContrast}
  n8 = ComponentMask(None=MaterialFunctionCall(In=Abs(None=@n6.R), Contrast=@n7) {materialFunction=&CheapContrast}.Result) {r=true}
  n9 = Custom(x=@MaterialExpressionMaterialFunctionCall_12.R, y=@MaterialExpressionMaterialFunctionCall_12.R) {
   code="return (atan2(y,x));"
   description=atan2
   outputType=CMOT_Float1
   inputs [
    {
     inputName=x
     input {expression=&self:MaterialExpressionMaterialFunctionCall_12}
    }
    {
     inputName=y
     input {
      expression=&self:MaterialExpressionMaterialFunctionCall_12
      outputIndex=1
     }
    }
   ]
  }
  n10 = MaterialFunctionCall(In=Abs(None=@n6.B), Contrast=@n7) {materialFunction=&CheapContrast}
  n11 = ComponentMask(None=LinearInterpolate(A=LinearInterpolate(A=@n3, B=@n4, Alpha=@n8), B=@n9, Alpha=ComponentMask(None=@n10.Result) {r=true})) {r=true}
  n12 = Constant() {r=6.2832}
  n13 = NamedRerouteUsage()
  n14 = AppendVector(A=Frac(None=Divide(A=@n11, B=@n12)), B=SphereMask(A=@n2, B=Constant(), Radius=Multiply(A=@n13) {constB=2}) {hardnessPercent=0})
  n15 = VectorParameter() {parameterName=StreaksTile, defaultValue=<rgba 3 3 0 0>}
  n16 = AppendVector(A=@n15.R, B=@n15.R)
  n17 = ScalarParameter() {defaultValue=0.1, group=Streaks, parameterName=Streaks Offset X}
  n18 = ScalarParameter() {defaultValue=0.5, group=Streaks, parameterName=Streaks Offset Y}
  n19 = TextureCoordinate()
  n20 = TextureSampleParameter2D!(
   UVs=Panner(Coordinate=Multiply(A=@n14, B=@n16), Speed=AppendVector(A=@n17, B=@n18)) {speedX=0.1, speedY=0.5}
   "DDX(UVs)"=DDX(Value=@n19)
   "DDY(UVs)"=DDY(Value=@n19)
   Apply View MipBias
  ) {
   mipValueMode=TMVM_Derivative
   parameterName=WaveTexture
   samplerType=SAMPLERTYPE_LinearColor
   texture=/Game/TechArt/Seasonal/LackLuster/Textures/T_Lackluster_Streaks
  }
  n21 = Multiply(A=@n20.G, B=ScalarParameter() {defaultValue=-0.25}) {constB=0.5}
  n22 = Fresnel(Normal=MaterialFunctionCall(Normal=NamedRerouteUsage(), Flatness=ScalarParameter() {defaultValue=0.25, parameterName=Param_1}) {materialFunction=&FlattenNormal}.Result) {baseReflectFraction=0, exponent=1.5}
  n23 = Frac(None=Multiply(A=Add(A=@n21, B=@n22), B=ScalarParameter() {defaultValue=1.25, group=HueShift, parameterName=HueShiftScale}))
  n24 = VectorParameter() {group=HueShift, parameterName=HueShiftColor, defaultValue=<rgba 0.2135 0.5727 1 0>}
  n25 = NamedRerouteUsage()
  n26 = TextureSample(UVs=SphereMask(A=@n2, B=Constant(), Radius=@n13, Hardness=LinearInterpolate(Alpha=@n25) {constA=-2, constB=0.8}) {hardnessPercent=0}) {texture=/Game/Effects/Fort_Effects/Textures/Masks/T_Vert_Gradient_Softer}
  n27 = Multiply(
   A=@n26.G
   B=LinearInterpolate(
    A=@n20.G
    B=MaterialFunctionCall(Hue Shift Percentage=@n23, Texture=@n24.RGB) {materialFunction=&HueShift}.Result
   )
  )
  n28 = Power(Base=@n26.G) {constExponent=1.6}
  MaterialFunctionCall(Emissive=Multiply(A=@n1.RGB, B=ShadingPathSwitch(Default=@n27, Forward=@n28, Mobile=@n28))) {materialFunction=&MF_FX_EmissiveAdjustments}
  MaterialFunctionCall() {materialFunction=&DebugTimeSine}
  Multiply(A=@n1.A, B=ShadingPathSwitch(Default=Multiply(A=@n26.G, B=@n20.G), Forward=@n26.G, Mobile=@n26.G))
  n29 = DynamicParameter() {defaultValue=<rgba 400 0 1 1>, paramNames [MeshScale,Erode,Param3,Param4]}
  NamedRerouteDeclaration(None=@n29.MeshScale) {name=MeshScaleParam, nodeColor=<rgba 1 0.8438 0 1>}
  NamedRerouteDeclaration(None=@n29.Erode) {name=ErodeParam, nodeColor=<rgba 0 0.8125 1 1>}
  TextureSample() {samplerType=SAMPLERTYPE_LinearGrayscale, texture=/Game/Effects/Fort_Effects/Textures/Tile/T_CellNoise_Tile_M}
 ]
}
