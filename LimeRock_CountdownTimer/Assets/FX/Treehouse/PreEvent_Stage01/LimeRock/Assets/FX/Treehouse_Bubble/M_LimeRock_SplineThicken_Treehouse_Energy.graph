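# Translucent unlit, two-sided Niagara material for animated energy strands in the LimeRock countdown Treehouse bubble effect.
# SplineThicken sets strand width, added to a time-driven rotation about an instance-space axis for world position offset.
# Panned, offset electricity samples plus a rainbow texture drive emissive, tinted by particle color and dynamic-parameter-blended gradients.
# Opacity combines remapped electricity, UV edge falloff, depth, view-angle and horizontal camera-distance fades, the last also gating width.
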Material /LimeRock_CountdownTimer/Assets/FX/Treehouse/PreEvent_Stage01/LimeRock/Assets/FX/Treehouse_Bubble/M_LimeRock_SplineThicken_Treehouse_Energy
graph {
 properties {
  bUsedWithNiagaraMeshParticles=true
  bUsedWithNiagaraRibbons=true
  blendMode=BLEND_Translucent
  pixelDepthOffsetMode=PDOM_Legacy
  shadingModel=MSM_Unlit
  twoSided=true
  usedShadingModels=Unlit
 }
 nodes [
  n1 = ActorPositionWS()
  n2 = Constant3Vector() {constant=<rgba 1 1 0 0>}
  n3 = CameraPositionWS()
  n4 = OneMinus(None=SphereMask(
    A=Multiply(A=@n1, B=@n2)
    B=Multiply(A=@n2, B=@n3)
    Radius=ScalarParameter() {defaultValue=1, group=Near Fade, parameterName=Fade_Distance}
    Hardness=ScalarParameter() {group=Near Fade, parameterName=hardness}
   ) {hardnessPercent=10})
  n5 = Ceil(None=@n4)
  n6 = DynamicParameter() {defaultValue=<rgba 1 1 0 1>, paramNames [SplineWidthMult, WidthTipMult, Color_Lerp, Opacity_Vertical_Power]}
  n7 = Multiply(A=ScalarParameter() {defaultValue=200, parameterName=Width}, B=@n6.SplineWidthMult)
  n8 = LinearInterpolate(
   A=@n7
   B=Multiply(A=ScalarParameter() {defaultValue=80, parameterName=WidthTip}, B=@n6.WidthTipMult)
   Alpha=ScalarParameter() {defaultValue=1, parameterName=Use Witdh Tip}
  )
  n9 = Transform(None=VectorParameter() {group=Rotation, parameterName=NormalizedRotationAxis, defaultValue=<rgba 1 0 0 0>}.RGB) {transformSourceType=TRANSFORMSOURCE_Instance}
  n10 = Time() {period=5}
  n11 = ScalarParameter() {group=Rotation, parameterName=Rotation_speed}
  n12 = VectorParameter() {group=Rotation, parameterName=PivotPointBias}
  Add(
   A=MaterialFunctionCall(
    WidthBase=Multiply(A=@n5, B=@n7)
    WidthTip=Multiply(A=@n5, B=@n8)
    Expand U or V UV Channel=StaticBoolParameter() {parameterName="Expand U or V UV Channel (SB)"}
   ) {materialFunction=&SplineThicken}.WorldPositionOffset
   B=RotateAboutAxis(
    NormalizedRotationAxis=Normalize(VectorInput=@n9)
    RotationAngle=Multiply(A=@n10, B=@n11)
    PivotPoint=TransformPosition(None=@n12.RGB) {transformSourceType=TRANSFORMPOSSOURCE_Instance, transformType=TRANSFORMPOSSOURCE_World}
    Position=WorldPosition().XYZ
   )
  ) {constB=0}
  n13 = Power(Base=ComponentMask(None=TextureCoordinate()) {r=true}, Exp=ScalarParameter() {defaultValue=1, parameterName=Color_Power})
  n14 = LinearInterpolate(
   A=VectorParameter() {parameterName=Emissive_Color_01, defaultValue=<rgba 1 0 0 1>}.RGB
   B=VectorParameter() {parameterName=Emissive_Color_02, defaultValue=<rgba 0 1 0 1>}.RGB
   Alpha=@n13
  )
  n15 = LinearInterpolate(
   A=VectorParameter() {parameterName=Emissive_Color_03, defaultValue=<rgba 1 0 0 1>}.RGB
   B=VectorParameter() {parameterName=Emissive_Color_04, defaultValue=<rgba 0 1 0 1>}.RGB
   Alpha=@n13
  )
  n16 = DynamicParameter() {defaultValue=<rgba 1 1 0 1>, paramNames [SplineWidthMult, WidthTipMult, Color_Lerp, Opacity_Vertical_Power]}
  n17 = Panner(Coordinate=Add(A=TextureCoordinate() {vTiling=0.5}, B=Constant2Vector() {g=0.1, r=0.5}), Time=NamedRerouteUsage()) {speedY=0.5}
  n18 = NamedRerouteUsage()
  n19 = ScalarParameter() {defaultValue=1, parameterName=Tiling X}
  n20 = ScalarParameter() {defaultValue=1, parameterName=Tiling Y}
  n21 = Add(A=Multiply(A=TextureCoordinate(), B=AppendVector(A=@n19, B=@n20)), B=AppendVector(A=Constant() {r=0.02}, B=Constant() {r=0.01}))
  n22 = TextureSample(UVs=Panner(Coordinate=Add(A=@n18, B=@n21), Time=NamedRerouteUsage()) {speedX=0.5}) {
   samplerType=SAMPLERTYPE_LinearColor
   texture=&/T_LimeRock_Electro_RGB
  }
  n23 = NamedRerouteUsage()
  n24 = Add(A=Multiply(A=TextureCoordinate(), B=AppendVector(A=@n19, B=@n20)), B=AppendVector(A=Constant(), B=Constant()))
  n25 = TextureSample(UVs=Panner(Coordinate=Add(A=@n23, B=@n24), Time=NamedRerouteUsage()) {speedX=0.5}) {
   samplerType=SAMPLERTYPE_LinearColor
   texture=&T_LimeRock_Electro_RGB
  }
  n26 = NamedRerouteUsage()
  n27 = Add(A=Multiply(A=TextureCoordinate(), B=AppendVector(A=@n19, B=@n20)), B=AppendVector(A=Constant() {r=-0.02}, B=Constant() {r=-0.01}))
  n28 = TextureSample(UVs=Panner(Coordinate=Add(A=@n26, B=@n27), Time=NamedRerouteUsage()) {speedX=0.5}) {
   samplerType=SAMPLERTYPE_LinearColor
   texture=&T_LimeRock_Electro_RGB
  }
  n29 = Multiply(A=Add(A=@n22.G, B=@n25.G), B=Add(A=@n25.G, B=@n28.G))
  n30 = Add(A=Power(Base=TextureSample(UVs=@n17) {texture=&T_Tile_RainbowColor}.RGB), B=@n29)
  n31 = ParticleColor()
  MaterialFunctionCall(Emissive=Multiply(A=LinearInterpolate(A=@n14, B=@n15, Alpha=@n16.Color_Lerp), B=Multiply(A=@n30, B=@n31.RGB))) {materialFunction=&MF_FX_EmissiveAdjustments}
  n32 = Constant2Vector() {g=0.1}
  n33 = MaterialFunctionCall!(Input=Desaturation(None=@n29), Input Low=@n32.R, Input High=@n32.R) {materialFunction=&/MF_LimeRock_RemapValue_0_1}
  n34 = ComponentMask(None=TextureCoordinate()) {r=true}
  n35 = Power(Base=Sine(None=@n34) {period=2}, Exp=ScalarParameter() {defaultValue=1, group=Opacity, parameterName=EdgeFade_Power})
  n36 = Multiply(
   A=ScalarParameter() {defaultValue=0.8, parameterName=Opacity}
   B=Multiply(A=Multiply(A=@n33.Result, B=@n35), B=Power(Base=@n34, Exp=@n6.Opacity_Vertical_Power))
  )
  n37 = DepthFade(Opacity=@n36, FadeDistance=ScalarParameter() {defaultValue=10, parameterName=FadeDistance}) {fadeDistanceDefault=10}
  n38 = ShadingPathSwitch(Default=@n37, Deferred=@n37, Forward=@n36, Mobile=@n36)
  n39 = ScalarParameter() {parameterName="FadeSensitivity (optional) (S)"}
  n40 = Saturate(None=Multiply(A=MaterialFunctionCall("FadeSensitivity (optional)"=@n39) {materialFunction=&FadeBasedOnViewAngle}.Result) {constB=10})
  Multiply(A=@n31.A, B=Multiply(A=Multiply(A=@n38, B=@n40), B=@n4))
  Multiply(A=VectorParameter() {parameterName=Color1, defaultValue=<rgba 2 0 0.6895 1>}.RGB, B=@n22.G)
  Multiply(A=@n28.G, B=VectorParameter() {parameterName=Color2, defaultValue=<rgba 0.0048 0.2333 2 1>}.RGB)
  NamedRerouteDeclaration(None=Multiply(A=Time(), B=ScalarParameter() {defaultValue=0.05, parameterName=Speed Direction}) {constB=-1.5}) {name=Speed, nodeColor=<rgba 1 0 0.9844 1>}
  n41 = Panner(Coordinate=TextureCoordinate() {uTiling=2, vTiling=0.5}) {speedY=0.2}
  NamedRerouteDeclaration(None=Multiply(A=TextureSample(UVs=@n41) {texture=&/T_LimeRock_Noise34}.R) {constB=0.03}) {name=Distortion, nodeColor=<rgba 1 0.375 0 1>}
  ScalarParameter() {parameterName=Opacity_Vertical_Power}
  ScalarParameter() {parameterName=Color_Lerp}
 ]
}
